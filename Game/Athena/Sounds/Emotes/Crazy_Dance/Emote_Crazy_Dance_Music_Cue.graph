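# Sound cue for the first-person music of the Crazy Dance emote.
# A mixer plays a one-shot intro alongside a looping track that starts after a fixed short delay, so the intro leads into the loop.
# It uses the first-person emote music sound class and concurrency set, and sends to a music audio-analysis submix.
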SoundCue /Game/Athena/Sounds/Emotes/Crazy_Dance/Emote_Crazy_Dance_Music_Cue
root=Mixer_0
nodes [
 {id=Delay_0, children [WavePlayer_0], properties {delayMax=0.69, delayMin=0.69}}
 {id=Mixer_0, children [Delay_0,WavePlayer_2], properties {inputVolume [1,1]}}
 {id=WavePlayer_0, properties {bLooping=true, soundWaveAssetPtr=&/Emote_Crazy_Dance_Loop}}
 {id=WavePlayer_2, properties {soundWaveAssetPtr=&/Emote_Crazy_Dance_Intro}}
]
properties {
 maxDistance=2097152
 soundClassObject=/Game/Sounds/FortSoundClassesAndModes/Fort_Music_Emote_1P_PSM
 volumeMultiplier=1
 concurrencyOverrides {resolutionRule=PreventNew}
 concurrencySet [/Game/Sounds/Concurrency_Rules/Athena_EmoteMusic_1P_Conc]
 soundSubmixSends [{minSendLevel=0, soundSubmix=&Fortnite_AudioAnalysis_Music}]
}
